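# Niagara system for the Noble emote that adds a flickering blue sprite glow to a holographic skeleton, judging by its name.
# Corona burst-spawns sprites and pins them each update to bone/socket indices 0-8 of User.SkeletalMesh, picked by particle execution index.
# CoronaLOW adds a larger glow at index 9 with a position offset.
# Both pulse alpha with a waveform and fade in and out over lifetime via curves.
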NiagaraSystem /BRCosmetics/Animation/Game/MainPlayer/Emotes/Noble/FX/NS_Noble_HoloSkelGlow
dependencies {
 usedDataInterfaces [&NiagaraDataInterfaceArrayInt32, &NiagaraDataInterfaceCurve]
 usedDynamicInputs []
 usedModules [&SystemState]
}
emitters {
 Corona {
  emitterUpdateScript [
   &EmitterState()
   &SpawnBurst_Instantaneous(Spawn Count=11, Use Spawn Probability=false, UseLoopCountLimit=false)
  ]
  particleSpawnScript [
   &InitializeParticle(
    Color=<rgba 0.01 0.2651 1 0.1>
    Lifetime=2.4
    Position Mode=::NewEnumerator1
    Sprite Size Mode=::NewEnumerator3
    Uniform Sprite Size=45
    UsePositionOffset=false
    Write Material Random=true
   )
   &DynamicMaterialParameters(
    Index 0 Param 1=2
    Module.Param0WriteEnabled=true
    Module.Param1WriteEnabled=true
    Module.Param2WriteEnabled=true
    Module.Param3WriteEnabled=true
   )
  ]
  particleUpdateScript [
   &ParticleState()
   /SetVariables(Particles.SpriteFacingFlat:Vector3f=<1 0 0>) {isSetParameters=true, module=SetVariables_#1}
   &SkeletalMeshLocation(
    Bone / Socket Index=&SelectIntFromArray(
     Direct Array Index=&ReturnExecIndex()
     Int Selection Array=DataInterface() {propertyValues {IntData [0,1,2,3,4,5,6,7,8]}}
    )
    MaskSpawnGroup=false
    Mesh Sampling Type=::NewEnumerator5
    Skeletal Mesh=$User.SkeletalMesh
    Socket Sampling Mode=::NewEnumerator5
   )
   &ScaleColor(
    Scale Alpha=&Waveform(
     Waveform Count=::NewEnumerator1
     "Waveform [1]"=::NewEnumerator0
     "[1] Frequency"=4
     "[1] Amplitude Min/Max"=<0 1>
     "[1] Scale"=0.2
     "[1] Bias"=1
     "[2] Average"=false
     "[2] Input Position"=0
     "[2] Phase"=$Emitter.Age
     "[2] Phase Scale"=1
     "[2] Compound Frequency"=2
     "[2] Frequency"=3
     "[2] Amplitude Min/Max"=<0 1>
     "[2] Scale"=0.2
     "[2] Bias"=0
     "[2] Exponent"=1
     "[2] Pulse Duration"=0.1
     "[2] Clip Triangle Min/Max"=<0 0>
     "[2] Random Seed"=0
    )
    ScaleA=true
    ScaleRGB=true
   )
   &ScaleColor(
    Scale Alpha=&FloatFromCurve(
     FloatCurve=DataInterface() {propertyValues {Curve {keys [
         {arriveTangent=1.3552, interpMode=RCIM_Cubic, leaveTangent=1.3552, tangentMode=RCTM_User, time=0, value=0}
         <at 0.2 1 RCIM_Cubic>
         <at 0.9 1 RCIM_Cubic>
         {arriveTangent=-2.0123, interpMode=RCIM_Cubic, leaveTangent=-2.0123, tangentMode=RCTM_User, time=1, value=0}
        ]}}}
     Scale Curve=0.8
    )
    ScaleA=true
    ScaleRGB=true
   ) {module=ScaleColor001}
  ]
  rendererClasses [&NiagaraSpriteRendererProperties]
  renderers [{rendererClass=&NiagaraSpriteRendererProperties}]
 }
 CoronaLOW {
  emitterUpdateScript [
   &EmitterState()
   &SpawnBurst_Instantaneous(Use Spawn Probability=false, UseLoopCountLimit=false)
  ]
  particleSpawnScript [
   &InitializeParticle(
    Color=<rgba 0.01 0.2651 1 0.3>
    Lifetime=2.4
    Position Mode=::NewEnumerator1
    Sprite Size Mode=::NewEnumerator3
    Uniform Sprite Size=65
    UsePositionOffset=false
    Write Material Random=true
   )
   &DynamicMaterialParameters(
    Index 0 Param 1=2
    Module.Param0WriteEnabled=true
    Module.Param1WriteEnabled=true
    Module.Param2WriteEnabled=true
    Module.Param3WriteEnabled=true
   )
  ]
  particleUpdateScript [
   &ParticleState()
   /SetVariables(Particles.SpriteFacingFlat:Vector3f=<1 0 0>) {enabled=false, isSetParameters=true, module=SetVariables_#1}
   &SkeletalMeshLocation(
    Bone / Socket Index=9
    MaskSpawnGroup=false
    Mesh Sampling Type=::NewEnumerator5
    Sampled Position Offset=<30 0 3>
    Skeletal Mesh=$User.SkeletalMesh
    Socket Sampling Mode=::NewEnumerator5
   )
   &ScaleColor(
    Scale Alpha=&Waveform(
     Waveform Count=::NewEnumerator1
     "Waveform [1]"=::NewEnumerator0
     "[1] Frequency"=4
     "[1] Amplitude Min/Max"=<0 1>
     "[1] Scale"=0.2
     "[1] Bias"=1
     "[2] Average"=false
     "[2] Input Position"=0
     "[2] Phase"=$Emitter.Age
     "[2] Phase Scale"=1
     "[2] Compound Frequency"=2
     "[2] Frequency"=3
     "[2] Amplitude Min/Max"=<0 1>
     "[2] Scale"=0.2
     "[2] Bias"=0
     "[2] Exponent"=1
     "[2] Pulse Duration"=0.1
     "[2] Clip Triangle Min/Max"=<0 0>
     "[2] Random Seed"=0
    )
    ScaleA=true
    ScaleRGB=true
   )
   &ScaleColor(
    Scale Alpha=&FloatFromCurve(FloatCurve=DataInterface() {propertyValues {Curve {keys [
         {arriveTangent=1.3552, interpMode=RCIM_Cubic, leaveTangent=1.3552, tangentMode=RCTM_User, time=0, value=0}
         <at 0.1 1 RCIM_Cubic>
         {arriveTangent=-1.1111, interpMode=RCIM_Cubic, leaveTangent=-1.1111, time=0.9, value=1}
         {arriveTangent=-2.0123, interpMode=RCIM_Cubic, leaveTangent=-2.0123, tangentMode=RCTM_User, time=1, value=0}
        ]}}})
    ScaleA=true
    ScaleRGB=true
   ) {module=ScaleColor001}
  ]
  rendererClasses [&NiagaraSpriteRendererProperties]
  renderers [{rendererClass=&NiagaraSpriteRendererProperties}]
 }
}
summary {
 systemName=NS_Noble_HoloSkelGlow
 emitters [
  {emitterName=Corona, rendererClasses [&NiagaraSpriteRendererProperties]}
  {emitterName=CoronaLOW, rendererClasses [&NiagaraSpriteRendererProperties]}
 ]
 userVariables [User.SkeletalMesh {
   type=NiagaraDataInterfaceSkeletalMesh
   defaultValue {
    struct=NiagaraExt_VariableValue_DataInterface
    value {dataInterface=&self:NiagaraDataInterfaceSkeletalMesh_0}
   }
  }]
}
